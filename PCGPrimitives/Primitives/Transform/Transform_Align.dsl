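; PCG primitive graph Transform_Align, which aligns input data using per-axis X, Y and Z justify-from and justify-to settings, with a UseParentAxis option.
; It tags primary and secondary inputs as Input:0 and Input:1, gathers them, and passes them through the shared template's inputs management.
; A core-process subgraph performs the alignment and feeds output preparation.
; Input processing, fallback, core and output subgraphs are exposed as swappable graph parameters.
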(graph Transform_Align
  (params (InputsProcessingGraph :type (SoftObject /Script/PCG.PCGGraphInterface) :default /PCGPrimitives/Primitives/_Template_/Primitives_InputsProcessing.Primitives_InputsProcessing) (InputsFallbackGraph :type (SoftObject /Script/PCG.PCGGraphInterface) :default /PCGPrimitives/Primitives/_Template_/Primitives_InputsFallback.Primitives_InputsFallback) (Justify_X_from :type (Enum /PCGPrimitives/Primitives/Transform/Subgraphs/EPCGJustifyFrom.EPCGJustifyFrom)) (Justify_X_to :type (Enum /PCGPrimitives/Primitives/Transform/Subgraphs/EPCGJustifyTo.EPCGJustifyTo)) (Justify_Y_from :type (Enum /PCGPrimitives/Primitives/Transform/Subgraphs/EPCGJustifyFrom.EPCGJustifyFrom)) (Justify_Y_to :type (Enum /PCGPrimitives/Primitives/Transform/Subgraphs/EPCGJustifyTo.EPCGJustifyTo)) (Justify_Z_from :type (Enum /PCGPrimitives/Primitives/Transform/Subgraphs/EPCGJustifyFrom.EPCGJustifyFrom)) (Justify_Z_to :type (Enum /PCGPrimitives/Primitives/Transform/Subgraphs/EPCGJustifyTo.EPCGJustifyTo)) (CoreProcessGraph :type (SoftObject /Script/PCG.PCGGraphInterface) :default /PCGPrimitives/Primitives/Transform/Subgraphs/Transform_Align_CoreProcess.Transform_Align_CoreProcess) (PrepareOutputsGraph :type (SoftObject /Script/PCG.PCGGraphInterface) :default /PCGPrimitives/Primitives/_Template_/Primitives_PrepareOutputs.Primitives_PrepareOutputs) (UseParentAxis :type Bool))
  (inputs (PrimaryInput :type Any) (SecondaryInput :type Any))
  (bind AddTags_1 (Add%20Tags :TagsToAdd "Input:0"))
  (bind AddTags_2 (Add%20Tags :TagsToAdd "Input:1"))
  (bind Gather_0 (Gather))
  ;[0 0 0 0.5 32] Inputs
  (bind Subgraph_0 (Subgraph :Graph /PCGPrimitives/Primitives/_Template_/Subgraphs/Primitives_InputsManagement.Primitives_InputsManagement :InputsFallbackGraph InputsFallbackGraph :InputsProcessingGraph InputsProcessingGraph))
  ;[0 0 0 0.5 32] Core Process
  (bind Subgraph_3 (Subgraph :Graph /PCGPrimitives/Primitives/Transform/Subgraphs/Transform_Align_CoreProcess.Transform_Align_CoreProcess :Justify_X_from Justify_X_from :Justify_X_to Justify_X_to :Justify_Y_from Justify_Y_from :Justify_Y_to Justify_Y_to :Justify_Z_from Justify_Z_from :Justify_Z_to Justify_Z_to :SubgraphOverride-param CoreProcessGraph :UseParentAxis UseParentAxis))
  ;[0 0 0 0.5 32] Prepare Output
  (bind Subgraph_4 (Subgraph :Graph /PCGPrimitives/Primitives/_Template_/Primitives_PrepareOutputs.Primitives_PrepareOutputs :SubgraphOverride-param PrepareOutputsGraph))
  (connect DefaultInputNode PrimaryInput AddTags_1 In)
  (connect DefaultInputNode SecondaryInput AddTags_2 In)
  (connect AddTags_1 Out Gather_0 In)
  (connect AddTags_2 Out Gather_0 In2)
  (connect Gather_0 Out Subgraph_0 In)
  (connect Subgraph_0 Out Subgraph_3 In)
  (connect Subgraph_3 Out Subgraph_4 In)
  (output Out Subgraph_4))
