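; PCG primitive graph that contracts areas by a Contract distance, with an optional union before output.
; It tags the primary input as "Input:0", then chains inputs management, a contraction core subgraph and an output preparation subgraph.
; The default input processing graph converts closed splines and forwards Poly2Ds.
; Input processing, input fallback, core and output graphs are exposed as parameters, so each stage can be swapped.
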(graph Transform_Areas_Contract
  (params (InputsProcessingGraph :type (SoftObject /Script/PCG.PCGGraphInterface) :default /PCGPrimitives/Primitives/Shared/InputsProcessing/InputsProcessing_ConvertClosedSplinesAndForwardPoly2Ds.InputsProcessing_ConvertClosedSplinesAndForwardPoly2Ds) (InputsFallbackGraph :type (SoftObject /Script/PCG.PCGGraphInterface) :default /PCGPrimitives/Primitives/_Template_/Primitives_InputsFallback.Primitives_InputsFallback) (CoreProcessGraph :type (SoftObject /Script/PCG.PCGGraphInterface) :default /PCGPrimitives/Primitives/Transform/Subgraphs/Transform_Areas_Contract_CoreProcess.Transform_Areas_Contract_CoreProcess) (PrepareOutputsGraph :type (SoftObject /Script/PCG.PCGGraphInterface) :default /PCGPrimitives/Primitives/_Template_/Primitives_PrepareOutputs.Primitives_PrepareOutputs) (Contract :type Double :default 1e+02) (UnionBeforeOutputting :type Bool))
  (inputs (PrimaryInput :type Any))
  (bind AddTags_0 (Add%20Tags :TagsToAdd "Input:0"))
  ;[0 0 0 0.5 32] Inputs
  (bind Subgraph_0 (Subgraph :Graph /PCGPrimitives/Primitives/_Template_/Subgraphs/Primitives_InputsManagement.Primitives_InputsManagement :InputsFallbackGraph InputsFallbackGraph :InputsProcessingGraph InputsProcessingGraph))
  ;[0 0 0 0.5 32] Core Process
  (bind Subgraph_3 (Subgraph :Graph /PCGPrimitives/Primitives/Transform/Subgraphs/Transform_Areas_Contract_CoreProcess.Transform_Areas_Contract_CoreProcess :Contract Contract :SubgraphOverride-param CoreProcessGraph :Union UnionBeforeOutputting))
  ;[0 0 0 0.5 32] Prepare Output
  (bind Subgraph_4 (Subgraph :Graph /PCGPrimitives/Primitives/_Template_/Primitives_PrepareOutputs.Primitives_PrepareOutputs :SubgraphOverride-param PrepareOutputsGraph))
  (connect DefaultInputNode PrimaryInput AddTags_0 In)
  (connect AddTags_0 Out Subgraph_0 In)
  (connect Subgraph_0 Out Subgraph_3 In)
  (connect Subgraph_3 Out Subgraph_4 In)
  (output Out Subgraph_4))
