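; Transform_Project is a PCG primitive graph that projects positions, rotations and/or scales between its two inputs, with optional distance falloff that can be written to density.
; Primary and secondary inputs are tagged by slot, gathered into one stream, then passed through input management, the projection core and output preparation subgraphs.
; Processing, fallback, core and output graphs are exposed as parameters for swapping.
; Further options cover closed-spline-to-polygon conversion and a projected attribute.
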(graph Transform_Project
  (params (InputsProcessingGraph :type (SoftObject /Script/PCG.PCGGraphInterface) :default /PCGPrimitives/Primitives/_Template_/Primitives_InputsProcessing.Primitives_InputsProcessing) (InputsFallbackGraph :type (SoftObject /Script/PCG.PCGGraphInterface) :default /PCGPrimitives/Primitives/_Template_/Primitives_InputsFallback.Primitives_InputsFallback) (CoreProcessGraph :type (SoftObject /Script/PCG.PCGGraphInterface) :default /PCGPrimitives/Primitives/Transform/Subgraphs/Transform_Project_CoreProcess.Transform_Project_CoreProcess) (PrepareOutputsGraph :type (SoftObject /Script/PCG.PCGGraphInterface) :default /PCGPrimitives/Primitives/_Template_/Primitives_PrepareOutputs.Primitives_PrepareOutputs) (ProjectPositions :type Bool :default true) (ProjectRotations :type Bool) (ProjectScales :type Bool) (UseFalloff :type Bool) (StartFalloffDistance :type Double) (EndFalloffDistance :type Double :default 1e+03) (SetFalloffAsDensity :type Bool) (ConvertClosedSplineToPolygon2D :type Bool) (ProjectAttribute :type String))
  (inputs (PrimaryInput :type Any) (SecondaryInput :type Any))
  (bind AddTags_1 (Add%20Tags :TagsToAdd "Input:0"))
  (bind AddTags_2 (Add%20Tags :TagsToAdd "Input:1"))
  (bind Gather_0 (Gather))
  ;[0 0 0 0.5 32] Inputs
  (bind Subgraph_0 (Subgraph :Graph /PCGPrimitives/Primitives/_Template_/Subgraphs/Primitives_InputsManagement.Primitives_InputsManagement :InputsFallbackGraph InputsFallbackGraph :InputsProcessingGraph InputsProcessingGraph))
  ;[0 0 0 0.5 32] Core Process
  (bind Subgraph_3 (Subgraph :Graph /PCGPrimitives/Primitives/Transform/Subgraphs/Transform_Project_CoreProcess.Transform_Project_CoreProcess :ConvertClosedSplineToPolygon2D ConvertClosedSplineToPolygon2D :EndFalloffDistance EndFalloffDistance :ProjectAttribute ProjectAttribute :ProjectPositions ProjectPositions :ProjectRotations ProjectRotations :ProjectScales ProjectScales :SetFalloffAsDensity SetFalloffAsDensity :StartFalloffDistance StartFalloffDistance :SubgraphOverride-param CoreProcessGraph :UseFalloff UseFalloff))
  ;[0 0 0 0.5 32] Prepare Output
  (bind Subgraph_4 (Subgraph :Graph /PCGPrimitives/Primitives/_Template_/Primitives_PrepareOutputs.Primitives_PrepareOutputs :SubgraphOverride-param PrepareOutputsGraph))
  (connect DefaultInputNode PrimaryInput AddTags_1 In)
  (connect DefaultInputNode SecondaryInput AddTags_2 In)
  (connect AddTags_1 Out Gather_0 In)
  (connect AddTags_2 Out Gather_0 In2)
  (connect Gather_0 Out Subgraph_0 In)
  (connect Subgraph_0 Out Subgraph_3 In)
  (connect Subgraph_3 Out Subgraph_4 In)
  (output Out Subgraph_4))
